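# Animation Blueprint for the Turbo Ball backpack cosmetic that continuously spins its dyn_meteor bone.
# Each update derives a yaw and roll rotator from real time and stores it in Spin.
# Starting from ref pose, one Transform Bone node applies Spin in bone space.
# A second adds a local X offset, weighted by interpolated, scale-biased JumpAlpha, which is not declared in this Blueprint.
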AnimBlueprint /BRCosmetics/Accessories/FORT_Backpacks/Backpack_M_MED_Turbo_Ball/Meshes/M_MED_Turbo_Ball_Pack_AnimBP
target_skeleton=&/M_MED_Turbo_Ball_Pack_Skeleton
graphs [
 AnimGraph {nodes [
   n1 = "Transform (Modify) Bone - Bone: dyn_meteor"(
    ComponentPose=Local To Component(LocalPose=Local Space Ref Pose())
    Alpha=1
    Translation="0,0,0"
    Rotation=Get Spin() -> (Spin:Rotator)
    Scale="1,1,1"
   ) {rotationMode=BMM_Replace, rotationSpace=BCS_BoneSpace, alphaBoolBlend {blendOption=Linear}, boneToModify {boneName=dyn_meteor}}
   n2 = "Transform (Modify) Bone - Bone: dyn_meteor"(ComponentPose=@n1, Alpha=Get JumpAlpha() -> (JumpAlpha:float), Translation="-8,0,0", Rotation="0,0,0", Scale="1,1,1") {
    translationMode=BMM_Additive
    translationSpace=BCS_BoneSpace
    alphaBoolBlend {blendOption=Linear}
    alphaScaleBiasClamp {bInterpResult=true, bias=0.15, interpSpeedDecreasing=4, interpSpeedIncreasing=5, scale=0.75}
    boneToModify {boneName=dyn_meteor}
    translation {x=-8}
   }
   Output Pose(Result=Component To Local(ComponentPose=@n2)) {layerGroup=None}
   Get SpinnerRot() -> (SpinnerRot:Rotator)
  ]}
 EventGraph {nodes [
   Event BlueprintUpdateAnimation()
   n1 = float * float(A=float / float(A=GetRealTimeSeconds(), B=-720) {is=title K2Node_CallFunction}, B=25000)
   > Set Spin(Spin:Rotator=MakeRotator(Roll=@n1, Yaw=@n1)) -> (Output_Get:Rotator)
   TryGetPawnOwner()
  ]}
]
variables [Spin {type=Rotator}]
